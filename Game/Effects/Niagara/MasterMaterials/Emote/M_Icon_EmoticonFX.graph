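# Translucent unlit master material for Niagara emote icon effects on sprites, ribbons and mesh particles.
# One texture's RGBA channels supply four masked layers, each with its own color, contrast and tiled, scrolling UVs, plus global UV controls.
# A scrolling normal map distorts UVs, and dynamic parameters drive per-layer alpha and optional contrast.
# Layers composite over a background color; summed alpha takes a sphere mask and particle alpha, and emissive passes through MF_FX_EmissiveAdjustments.
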Material /Game/Effects/Niagara/MasterMaterials/Emote/M_Icon_EmoticonFX
graph {
 properties {
  bAutomaticallySetUsageInEditor=false
  bHasPixelAnimation=true
  bUsedWithNiagaraMeshParticles=true
  bUsedWithNiagaraRibbons=true
  bUsedWithNiagaraSprites=true
  blendMode=BLEND_Translucent
  shadingModel=MSM_Unlit
  usedShadingModels=Unlit
 }
 nodes [
  MaterialFunctionCall(Emissive=NamedRerouteUsage()) {materialFunction=&MF_FX_EmissiveAdjustments}
  n1 = TextureObjectParameter() {
   desc="Base Pattern Texture - RGBA matches with Layer 1,2,3,4"
   group=Base Texture
   parameterName=Icon Texture
   texture=/Game/Effects/Niagara/MasterMaterials/Emote/Textures/T_Icon_Emoticon_InteractionFX_Default
  }
  n2 = TextureSample(UVs=NamedRerouteUsage(), Tex=@n1) {texture=&T_Icon_Emoticon_InteractionFX_Default}
  n3 = MaterialFunctionCall(
   In=@n2.R
   Contrast=LinearInterpolate(
    A=ScalarParameter() {group=Layer 1, parameterName=Layer 1 Contrast, sortPriority=31}
    B=DynamicParameter() {parameterIndex=1, paramNames [Layer 1 Contrast, Layer 2 Contrast, Layer 3 Contrast, Layer 4 Contrast]}.Layer 1 Contrast
    Alpha=ScalarParameter() {group=Layer 1, parameterName=Layer 1 Use Particle Contrast}
   )
  ) {materialFunction=&CheapContrast}
  n4 = VectorParameter() {group=Color, parameterName=Layer 1 Color, sortPriority=1, defaultValue=<rgba 1 0.1599 0.2845 1>}
  n5 = ScalarParameter() {defaultValue=1, desc=Multiplies the mask into the color., group=Color, parameterName=Layer 1 - Mask Color Contribution, sortPriority=2}
  NamedRerouteDeclaration(None=LinearInterpolate(A=@n4.RGB, B=Multiply(A=@n4.RGB, B=@n3.Result), Alpha=@n5)) {name=Layer 1 Color, nodeColor=<rgba 0.3828 1 0 1>}
  NamedRerouteDeclaration(None=@n2.R) {name=Layer 1 Mask, nodeColor=<rgba 1 0 0.7031 1>}
  n6 = ScalarParameter() {defaultValue=1, group=Layer 1, parameterName=Layer 1 Tile U, sortPriority=34}
  n7 = ScalarParameter() {defaultValue=1, group=Layer 1, parameterName=Layer 1 Tile V, sortPriority=35}
  n8 = Add(
   A=Multiply(A=NamedRerouteUsage(), B=AppendVector(A=@n6, B=@n7))
   B=AppendVector(
    A=ScalarParameter() {group=Layer 1, parameterName=Layer 1 Offset U, sortPriority=36}
    B=ScalarParameter() {group=Layer 1, parameterName=Layer 1 Offset V, sortPriority=37}
   )
  )
  n9 = AppendVector(
   A=ScalarParameter() {group=Layer 1, parameterName=Layer 1 Scroll U, sortPriority=38}
   B=ScalarParameter() {group=Layer 1, parameterName=Layer 1 Scroll V, sortPriority=39}
  )
  n10 = NamedRerouteUsage()
  n11 = ScalarParameter() {group=Layer 1, parameterName=Layer 1 UV Distortion Strength, sortPriority=33}
  NamedRerouteDeclaration(None=Add(A=Panner(Coordinate=@n8, Speed=@n9), B=Multiply(A=@n10, B=@n11))) {name=Layer 1 UVs, nodeColor=<rgba 0 1 0.625 1>}
  n12 = ScalarParameter() {defaultValue=1, group=Layer 2, parameterName=Layer 2 Tile U, sortPriority=34}
  n13 = ScalarParameter() {defaultValue=1, group=Layer 2, parameterName=Layer 2 Tile V, sortPriority=35}
  n14 = Add(
   A=Multiply(A=NamedRerouteUsage(), B=AppendVector(A=@n12, B=@n13))
   B=AppendVector(
    A=ScalarParameter() {group=Layer 2, parameterName=Layer 2 Offset U, sortPriority=36}
    B=ScalarParameter() {group=Layer 2, parameterName=Layer 2 Offset V, sortPriority=37}
   )
  )
  n15 = AppendVector(
   A=ScalarParameter() {group=Layer 2, parameterName=Layer 2 Scroll U, sortPriority=38}
   B=ScalarParameter() {group=Layer 2, parameterName=Layer 2 Scroll V, sortPriority=39}
  )
  n16 = NamedRerouteUsage()
  n17 = ScalarParameter() {group=Layer 2, parameterName=Layer 2 UV Distortion Strength, sortPriority=33}
  NamedRerouteDeclaration(None=Add(A=Panner(Coordinate=@n14, Speed=@n15), B=Multiply(A=@n16, B=@n17))) {name=Layer 2 UVs, nodeColor=<rgba 0 1 0.625 1>}
  n18 = ScalarParameter() {defaultValue=1, group=Layer 3, parameterName=Layer 3 Tile U, sortPriority=34}
  n19 = ScalarParameter() {defaultValue=1, group=Layer 3, parameterName=Layer 3 Tile V, sortPriority=35}
  n20 = Add(
   A=Multiply(A=NamedRerouteUsage(), B=AppendVector(A=@n18, B=@n19))
   B=AppendVector(
    A=ScalarParameter() {group=Layer 3, parameterName=Layer 3 Offset U, sortPriority=36}
    B=ScalarParameter() {group=Layer 3, parameterName=Layer 3 Offset V, sortPriority=37}
   )
  )
  n21 = AppendVector(
   A=ScalarParameter() {group=Layer 3, parameterName=Layer 3 Scroll U, sortPriority=38}
   B=ScalarParameter() {group=Layer 3, parameterName=Layer 3 Scroll V, sortPriority=39}
  )
  n22 = NamedRerouteUsage()
  n23 = ScalarParameter() {group=Layer 3, parameterName=Layer 3 UV Distortion Strength, sortPriority=33}
  NamedRerouteDeclaration(None=Add(A=Panner(Coordinate=@n20, Speed=@n21), B=Multiply(A=@n22, B=@n23))) {name=Layer 3 UVs, nodeColor=<rgba 0 1 0.625 1>}
  n24 = NamedRerouteUsage()
  n25 = DynamicParameter() {paramNames [Layer 1 Alpha, Layer 2 Alpha, Layer 3 Alpha, Layer 4 Alpha]}
  n26 = LinearInterpolate(
   A=VectorParameter() {group=Color, parameterName=BG Color, sortPriority=1}.RGB
   B=NamedRerouteUsage()
   Alpha=Clamp(None=Multiply(A=@n24, B=@n25.Layer 4 Alpha))
  )
  n27 = Clamp(None=Multiply(A=@n25.Layer 3 Alpha, B=NamedRerouteUsage()))
  n28 = NamedRerouteUsage()
  n29 = Multiply(A=@n25.Layer 2 Alpha, B=NamedRerouteUsage())
  n30 = NamedRerouteUsage()
  n31 = LinearInterpolate(
   A=LinearInterpolate(A=LinearInterpolate(A=@n26, B=@n28, Alpha=@n27), B=NamedRerouteUsage(), Alpha=Clamp(None=@n29))
   B=NamedRerouteUsage()
   Alpha=Clamp(None=Multiply(A=@n25.Layer 1 Alpha, B=@n30))
  )
  n32 = ParticleColor()
  NamedRerouteDeclaration(None=LinearInterpolate(A=@n31, B=Multiply(A=@n31, B=@n32.RGB), Alpha=ScalarParameter() {group=Color, parameterName=Partical RGB Contribution, sortPriority=-1})) {name=Color, nodeColor=<rgba 0 0.8828 1 1>}
  n33 = Subtract(A=TextureCoordinate()) {constB=0.5}
  n34 = AppendVector(
   A=ScalarParameter() {defaultValue=1, group=Global UV, parameterName=Global Tile U, sortPriority=34}
   B=ScalarParameter() {defaultValue=1, group=Global UV, parameterName=Global Tile V, sortPriority=35}
  )
  n35 = AppendVector(
   A=ScalarParameter() {group=Global UV, parameterName=Global Offset U, sortPriority=36}
   B=ScalarParameter() {group=Global UV, parameterName=Global Offset V, sortPriority=37}
  )
  n36 = Add(A=Add(A=Multiply(A=@n33, B=@n34), B=Multiply(A=@n35, B=@n34))) {constB=0.5}
  n37 = AppendVector(
   A=ScalarParameter() {group=Global UV, parameterName=Global Scroll U, sortPriority=38}
   B=ScalarParameter() {group=Global UV, parameterName=Global Scroll V, sortPriority=39}
  )
  n38 = NamedRerouteUsage()
  n39 = ScalarParameter() {group=Global UV, parameterName=Global Distortion Strength, sortPriority=33}
  NamedRerouteDeclaration(None=Add(A=Panner(Coordinate=@n36, Speed=@n37), B=Multiply(A=@n38, B=@n39))) {name=Global UV, nodeColor=<rgba 0 1 0.625 1>}
  n40 = ScalarParameter() {defaultValue=1, group=Layer 4, parameterName=Layer 4 Tile U, sortPriority=34}
  n41 = ScalarParameter() {defaultValue=1, group=Layer 4, parameterName=Layer 4 Tile V, sortPriority=35}
  n42 = Add(
   A=Multiply(A=NamedRerouteUsage(), B=AppendVector(A=@n40, B=@n41))
   B=AppendVector(
    A=ScalarParameter() {group=Layer 4, parameterName=Layer 4 Offset U, sortPriority=36}
    B=ScalarParameter() {group=Layer 4, parameterName=Layer 4 Offset V, sortPriority=37}
   )
  )
  n43 = AppendVector(
   A=ScalarParameter() {group=Layer 4, parameterName=Layer 4 Scroll U, sortPriority=38}
   B=ScalarParameter() {group=Layer 4, parameterName=Layer 4 Scroll V, sortPriority=39}
  )
  n44 = NamedRerouteUsage()
  n45 = ScalarParameter() {group=Layer 4, parameterName=Layer 4 UV Distortion Strength, sortPriority=33}
  NamedRerouteDeclaration(None=Add(A=Panner(Coordinate=@n42, Speed=@n43), B=Multiply(A=@n44, B=@n45))) {name=Layer 4 UVs, nodeColor=<rgba 0 1 0.625 1>}
  n46 = TextureSample(UVs=NamedRerouteUsage(), Tex=@n1) {texture=&T_Icon_Emoticon_InteractionFX_Default}
  n47 = MaterialFunctionCall(
   In=@n46.G
   Contrast=LinearInterpolate(
    A=ScalarParameter() {group=Layer 2, parameterName=Layer 2 Contrast, sortPriority=31}
    B=DynamicParameter() {parameterIndex=1, paramNames [Layer 1 Contrast, Layer 2 Contrast, Layer 3 Contrast, Layer 4 Contrast]}.Layer 2 Contrast
    Alpha=ScalarParameter() {group=Layer 2, parameterName=Layer 2 Use Particle Contrast}
   )
  ) {materialFunction=&CheapContrast}
  n48 = VectorParameter() {group=Color, parameterName=Layer 2 Color, sortPriority=3, defaultValue=<rgba 0.2849 1 0.4975 1>}
  n49 = ScalarParameter() {defaultValue=1, desc=Multiplies the mask into the color., group=Color, parameterName=Layer 2 - Mask Color Contribution, sortPriority=4}
  NamedRerouteDeclaration(None=LinearInterpolate(A=@n48.RGB, B=Multiply(A=@n48.RGB, B=@n47.Result), Alpha=@n49)) {name=Layer 2 Color, nodeColor=<rgba 0.3828 1 0 1>}
  NamedRerouteDeclaration(None=@n46.G) {name=Layer 2 Mask, nodeColor=<rgba 1 0 0.7031 1>}
  n50 = TextureSample(UVs=NamedRerouteUsage(), Tex=@n1) {texture=&T_Icon_Emoticon_InteractionFX_Default}
  n51 = MaterialFunctionCall(
   In=@n50.B
   Contrast=LinearInterpolate(
    A=ScalarParameter() {group=Layer 3, parameterName=Layer 3 Contrast, sortPriority=31}
    B=DynamicParameter() {parameterIndex=1, paramNames [Layer 1 Contrast, Layer 2 Contrast, Layer 3 Contrast, Layer 4 Contrast]}.Layer 3 Contrast
    Alpha=ScalarParameter() {group=Layer 3, parameterName=Layer 3 Use Particle Contrast}
   )
  ) {materialFunction=&CheapContrast}
  n52 = VectorParameter() {group=Color, parameterName=Layer 3 Color, sortPriority=5, defaultValue=<rgba 0.2 0.2916 1 1>}
  n53 = ScalarParameter() {defaultValue=1, desc=Multiplies the mask into the color., group=Color, parameterName=Layer 3 - Mask Color Contribution, sortPriority=6}
  NamedRerouteDeclaration(None=LinearInterpolate(A=@n52.RGB, B=Multiply(A=@n52.RGB, B=@n51.Result), Alpha=@n53)) {name=Layer 3 Color, nodeColor=<rgba 0.3828 1 0 1>}
  NamedRerouteDeclaration(None=@n50.B) {name=Layer 3 Mask, nodeColor=<rgba 1 0 0.7031 1>}
  n54 = TextureSample(UVs=NamedRerouteUsage(), Tex=@n1) {texture=&T_Icon_Emoticon_InteractionFX_Default}
  n55 = MaterialFunctionCall(
   In=@n54.A
   Contrast=LinearInterpolate(
    A=ScalarParameter() {group=Layer 4, parameterName=Layer 4 Contrast, sortPriority=31}
    B=DynamicParameter() {parameterIndex=1, paramNames [Layer 1 Contrast, Layer 2 Contrast, Layer 3 Contrast, Layer 4 Contrast]}.Layer 4 Contrast
    Alpha=ScalarParameter() {group=Layer 4, parameterName=Layer 4 Use Particle Contrast}
   )
  ) {materialFunction=&CheapContrast}
  n56 = VectorParameter() {group=Color, parameterName=Layer 4 Color, sortPriority=7, defaultValue=<rgba 1 0.4676 0.1136 1>}
  n57 = ScalarParameter() {defaultValue=1, desc=Multiplies the mask into the color., group=Color, parameterName=Layer 4 - Mask Color Contribution, sortPriority=8}
  NamedRerouteDeclaration(None=LinearInterpolate(A=@n56.RGB, B=Multiply(A=@n56.RGB, B=@n55.Result), Alpha=@n57)) {name=Layer 4 Color, nodeColor=<rgba 0.3828 1 0 1>}
  NamedRerouteDeclaration(None=@n54.A) {name=Layer 4 Mask, nodeColor=<rgba 1 0 0.7031 1>}
  n58 = DynamicParameter() {paramNames [Layer 1 Alpha, Layer 2 Alpha, Layer 3 Alpha, Layer 4 Alpha]}
  n59 = Add(A=Multiply(A=@n58.Layer 1 Alpha, B=NamedRerouteUsage()), B=Multiply(A=@n58.Layer 2 Alpha, B=NamedRerouteUsage()))
  n60 = Multiply(A=@n58.Layer 3 Alpha, B=NamedRerouteUsage())
  n61 = NamedRerouteUsage()
  n62 = ScalarParameter() {defaultValue=0.5, group=Sphere Mask, parameterName=Sphere Mask Hardness, sortPriority=33}
  n63 = ScalarParameter() {defaultValue=0.5, group=Sphere Mask, parameterName=Sphere Mask Center, sortPriority=34}
  n64 = ScalarParameter() {defaultValue=0.5, group=Sphere Mask, parameterName=Sphere Mask Radius, sortPriority=35}
  n65 = Multiply(
   A=Add(A=Add(A=@n59, B=@n60), B=Multiply(A=@n58.Layer 4 Alpha, B=@n61))
   B=LinearInterpolate(
    A=Constant() {r=1}
    B=MaterialFunctionCall!(Hardness=@n62, Center=@n63, Radius=@n64) {materialFunction=/Game/Packages/Fortress_SharedMaterials/Base_Material_Functions/FX_SphereMask2D}.Result
    Alpha=ScalarParameter() {defaultValue=1, group=Sphere Mask, parameterName=Sphere Mask Contribution}
   )
  )
  NamedRerouteDeclaration(None=Multiply(A=Clamp(None=@n65), B=ParticleColor().A)) {name=Alpha, nodeColor=<rgba 0.9453 1 0 1>}
  n66 = Multiply(
   A=TextureCoordinate()
   B=AppendVector(
    A=ScalarParameter() {defaultValue=2, group=UV Distortion, parameterName=UV Distortion Tile U, sortPriority=34}
    B=ScalarParameter() {defaultValue=2, group=UV Distortion, parameterName=UV Distortion Tile V, sortPriority=35}
   )
  )
  n67 = AppendVector(
   A=ScalarParameter() {group=UV Distortion, parameterName=UV Distortion Offset U, sortPriority=36}
   B=ScalarParameter() {group=UV Distortion, parameterName=UV Distortion Offset V, sortPriority=37}
  )
  n68 = ScalarParameter() {defaultValue=0.1, group=UV Distortion, parameterName=UV Distortion Scroll U, sortPriority=38}
  n69 = ScalarParameter() {defaultValue=0.1, group=UV Distortion, parameterName=UV Distortion Scroll V, sortPriority=39}
  n70 = TextureSampleParameter2D(UVs=Panner(Coordinate=Add(A=@n66, B=@n67), Speed=AppendVector(A=@n68, B=@n69))) {
   desc=Normal Map for UV distortion
   group=UV Distortion
   parameterName=UV Distortion Texture
   samplerType=SAMPLERTYPE_Normal
   texture=/Game/Effects/Niagara/MasterMaterials/Emote/Textures/T_FX_UVDistortion_01
  }
  n71 = ScalarParameter() {defaultValue=1, desc=Global Strength for UV distortion, group=UV Distortion, parameterName=UV Distortion Strength, sortPriority=33}
  NamedRerouteDeclaration(None=Multiply(A=ComponentMask(None=@n70.RGB) {g=true, r=true}, B=@n71)) {name=UV Distortion, nodeColor=<rgba 0.1484 0 1 1>}
  NamedRerouteUsage()
 ]
}
